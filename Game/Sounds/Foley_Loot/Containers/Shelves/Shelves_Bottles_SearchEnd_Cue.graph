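# Sound cue for the end of searching a shelf container with bottles, mixing a no-more-loot UI sound with shelf foley.
# The foley layers one random jar-food take under one random bottle take at full level.
# An enveloper fades that pair out and a modulator varies its pitch and volume before the final mix.
# Search-prop attenuation and a concurrency override that stops the farthest instance first govern playback in the world.
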SoundCue /Game/Sounds/Foley_Loot/Containers/Shelves/Shelves_Bottles_SearchEnd_Cue
root=Mixer_11
nodes [
 {
  id=Enveloper_9
  children [Mixer_3]
  properties {pitchCurve {editorCurveData {keys [{time=0}]}}, volumeCurve {editorCurveData {keys [<at 0 1>,<at 0.64 0>]}}}
 }
 {id=Mixer_11, children [WavePlayer_42, Modulator_14], properties {inputVolume [0.8,1]}}
 {id=Mixer_3, children [Random_17,Random_15], properties {inputVolume [0.5,1]}}
 {id=Modulator_14, children [Enveloper_9], properties {pitchMax=1.07, pitchMin=0.92, volumeMax=1.07, volumeMin=0.92}}
 {id=Random_15, children [WavePlayer_110, WavePlayer_113, WavePlayer_114], properties {weights [1,1,1]}}
 {id=Random_17, children [WavePlayer_97, WavePlayer_98, WavePlayer_99], properties {weights [1,1,1]}}
 {id=WavePlayer_110, properties {soundWaveAssetPtr=&/Shelf_Bottle_Search06}}
 {id=WavePlayer_113, properties {soundWaveAssetPtr=&/Shelf_Bottle_Search03}}
 {id=WavePlayer_114, properties {soundWaveAssetPtr=&/Shelf_Bottle_Search04}}
 {id=WavePlayer_42, properties {soundWaveAssetPtr=/Game/Sounds/Fort_UI_Sounds/Gameplay/Fort_UI_NoMoreLoot}}
 {id=WavePlayer_97, properties {soundWaveAssetPtr=&/Shelves_JarFood_Search1}}
 {id=WavePlayer_98, properties {soundWaveAssetPtr=&/Shelves_JarFood_Search13}}
 {id=WavePlayer_99, properties {soundWaveAssetPtr=&/Shelves_JarFood_Search14}}
]
properties {
 attenuationSettings=/Game/Sounds/Attenuation/Fort_Prop_Search_Att
 bOverrideConcurrency=true
 maxDistance=1300
 soundClassObject=&Fort_Loot_Collect
 volumeMultiplier=0.7
 concurrencyOverrides {maxCount=4, resolutionRule=StopFarthestThenPreventNew}
}
