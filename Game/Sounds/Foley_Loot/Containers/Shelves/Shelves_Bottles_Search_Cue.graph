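# Sound cue for searching shelves of bottles and food jars, a looting foley effect in the loot-collect sound class.
# A mixer layers one random jar-food search sound at half volume with one random bottle search sound at full volume.
# A modulator then varies pitch and volume slightly on each play.
# It uses prop-search attenuation, and its concurrency override prevents new instances from starting once the limit is reached.
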SoundCue /Game/Sounds/Foley_Loot/Containers/Shelves/Shelves_Bottles_Search_Cue
root=Modulator_24
nodes [
 {id=Mixer_2, children [Random_16,Random_14], properties {inputVolume [0.5,1]}}
 {id=Modulator_24, children [Mixer_2], properties {pitchMax=1.07, pitchMin=0.92, volumeMax=1.07, volumeMin=0.92}}
 {id=Random_14, children [WavePlayer_104, WavePlayer_107, WavePlayer_108], properties {weights [1,1,1]}}
 {id=Random_16, children [WavePlayer_93, WavePlayer_94, WavePlayer_95], properties {weights [1,1,1]}}
 {id=WavePlayer_104, properties {soundWaveAssetPtr=&/Shelf_Bottle_Search06}}
 {id=WavePlayer_107, properties {soundWaveAssetPtr=&/Shelf_Bottle_Search03}}
 {id=WavePlayer_108, properties {soundWaveAssetPtr=&/Shelf_Bottle_Search04}}
 {id=WavePlayer_93, properties {soundWaveAssetPtr=&/Shelves_JarFood_Search1}}
 {id=WavePlayer_94, properties {soundWaveAssetPtr=&/Shelves_JarFood_Search13}}
 {id=WavePlayer_95, properties {soundWaveAssetPtr=&/Shelves_JarFood_Search14}}
]
properties {
 attenuationSettings=/Game/Sounds/Attenuation/Fort_Prop_Search_Att
 bOverrideConcurrency=true
 maxDistance=1300
 soundClassObject=&Fort_Loot_Collect
 volumeMultiplier=0.7
 concurrencyOverrides {resolutionRule=PreventNew}
}
