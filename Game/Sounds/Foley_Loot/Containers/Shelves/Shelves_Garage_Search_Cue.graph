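# Sound cue for the loot foley played when searching garage shelves, using the loot collect sound class and prop search attenuation.
# A Random node picks among three search waves and feeds a Modulator for slight pitch and volume variation.
# A Mixer layers two branches sharing that chain: an Enveloper that starts boosted and settles over time, and a roughly one-second randomized Delay for a follow-up rummage.
# A concurrency override prevents new instances once the limit is reached.
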SoundCue /Game/Sounds/Foley_Loot/Containers/Shelves/Shelves_Garage_Search_Cue
root=Mixer_25
nodes [
 {id=Delay_45, children [Modulator_24], properties {delayMax=1.2, delayMin=1}}
 {
  id=Enveloper_4
  children [Modulator_24]
  properties {pitchCurve {editorCurveData {keys [{time=0}]}}, volumeCurve {editorCurveData {keys [<at 0.0001 1.5>,<at 1 1>]}}}
 }
 {id=Mixer_25, children [Enveloper_4,Delay_45], properties {inputVolume [1,1]}}
 {id=Modulator_24, children [Random_19], properties {pitchMax=1.07, pitchMin=0.92, volumeMax=1.07, volumeMin=0.92}}
 {id=Random_19, children [WavePlayer_104, WavePlayer_105, WavePlayer_106], properties {weights [1,1,1]}}
 {id=WavePlayer_104, properties {soundWaveAssetPtr=&/Shelves_Garage_Search4}}
 {id=WavePlayer_105, properties {soundWaveAssetPtr=&/Shelves_Garage_Search1}}
 {id=WavePlayer_106, properties {soundWaveAssetPtr=&/Shelves_Garage_Search2}}
]
properties {
 attenuationSettings=/Game/Sounds/Attenuation/Fort_Prop_Search_Att
 bOverrideConcurrency=true
 maxDistance=1300
 soundClassObject=&Fort_Loot_Collect
 volumeMultiplier=1.2
 concurrencyOverrides {resolutionRule=PreventNew}
}
