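# Sound cue for completing a search of a shelves jar-food container, routed to the loot collect sound class.
# It picks one of three search waves at equal weight, fades it out over about two-thirds of a second with an enveloper, then applies slight random pitch and volume modulation.
# Prop-search attenuation spatializes it, and a concurrency override caps simultaneous instances, stopping the farthest first.
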SoundCue /Game/Sounds/Foley_Loot/Containers/Shelves/Shelves_JarFood_SearchComplete_Cue
root=Modulator_14
nodes [
 {
  id=Enveloper_8
  children [Random_17]
  properties {pitchCurve {editorCurveData {keys [{time=0}]}}, volumeCurve {editorCurveData {keys [<at 0 1>,<at 0.64 0>]}}}
 }
 {id=Modulator_14, children [Enveloper_8], properties {pitchMax=1.07, pitchMin=0.92, volumeMax=1.07, volumeMin=0.92}}
 {id=Random_17, children [WavePlayer_96, WavePlayer_98, WavePlayer_99], properties {weights [1,1,1]}}
 {id=WavePlayer_96, properties {soundWaveAssetPtr=&/Shelves_JarFood_Search8}}
 {id=WavePlayer_98, properties {soundWaveAssetPtr=&/Shelves_JarFood_Search13}}
 {id=WavePlayer_99, properties {soundWaveAssetPtr=&/Shelves_JarFood_Search14}}
]
properties {
 attenuationSettings=/Game/Sounds/Attenuation/Fort_Prop_Search_Att
 bOverrideConcurrency=true
 maxDistance=1300
 soundClassObject=&Fort_Loot_Collect
 volumeMultiplier=0.7
 concurrencyOverrides {maxCount=4, resolutionRule=StopFarthestThenPreventNew}
}
