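# Sound cue for the foley heard when searching jarred food on shelf loot containers, classed as loot collect audio with prop search attenuation.
# A random pick from three search waves feeds a modulator that slightly varies pitch and volume.
# The mixer reuses that chain twice, once immediately and once quieter after a delay of about a second, layering a longer search from few waves.
# Its concurrency override prevents new instances at the limit.
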SoundCue /Game/Sounds/Foley_Loot/Containers/Shelves/Shelves_JarFood_Search_Cue
root=Mixer_24
nodes [
 {id=Delay_44, children [Modulator_24], properties {delayMax=1.2, delayMin=1}}
 {id=Mixer_24, children [Modulator_24,Delay_44], properties {inputVolume [1,0.7]}}
 {id=Modulator_24, children [Random_16], properties {pitchMax=1.07, pitchMin=0.92, volumeMax=1.07, volumeMin=0.92}}
 {id=Random_16, children [WavePlayer_92, WavePlayer_94, WavePlayer_95], properties {weights [1,1,1]}}
 {id=WavePlayer_92, properties {soundWaveAssetPtr=&/Shelves_JarFood_Search8}}
 {id=WavePlayer_94, properties {soundWaveAssetPtr=&/Shelves_JarFood_Search13}}
 {id=WavePlayer_95, properties {soundWaveAssetPtr=&/Shelves_JarFood_Search14}}
]
properties {
 attenuationSettings=/Game/Sounds/Attenuation/Fort_Prop_Search_Att
 bOverrideConcurrency=true
 maxDistance=1300
 soundClassObject=&Fort_Loot_Collect
 volumeMultiplier=1.3
 concurrencyOverrides {resolutionRule=PreventNew}
}
